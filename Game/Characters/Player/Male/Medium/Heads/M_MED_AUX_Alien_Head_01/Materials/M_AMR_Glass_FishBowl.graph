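# Translucent glass fishbowl material for the alien player head, showing water and rising bubbles.
# Three panning pre-skinned-position bubble layers, masked by a distortion texture and inverse Fresnel, drive water opacity, refraction and cubemap-lit highlights alongside a sun specular.
# A Fresnel-weighted attribute blend mixes this water with a glass layer that has noise normals and Fresnel opacity.
# Opacity is scaled by the character res-in effect, and quality switches simplify highlights, normals and refraction.
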Material /Game/Characters/Player/Male/Medium/Heads/M_MED_AUX_Alien_Head_01/Materials/M_AMR_Glass_FishBowl
graph {
 properties {
  bEnableResponsiveAA=true
  bScreenSpaceReflections=true
  bUseMaterialAttributes=true
  bUsedWithMeshDeformer=true
  bUsedWithNanite=true
  bUsedWithSkeletalMesh=true
  blendMode=BLEND_Translucent
  ditherOpacityMask=true
  physMaterial=/Game/FortressPhysicalMaterials/Glass
  pixelDepthOffsetMode=PDOM_Legacy
  refractionMethod=RM_IndexOfRefraction
  translucencyDirectionalLightingIntensity=5
  translucencyLightingMode=TLM_Surface
  usedShadingModels=Default Lit
 }
 nodes [
  MaterialExpressionTextureSampleParameterCube_0 = TextureSampleParameterCube!(UVs=ReflectionVectorWS() {bNormalizeCustomWorldNormal=true}, Apply View MipBias) {parameterName=MaterialExpressionTextureSampleParameterCube_0, texture=/Game/Packages/Fortress_Sky/TexturesHDR/T_hdrsky_test03_low}
  n1 = Multiply(
   A=@MaterialExpressionTextureSampleParameterCube_0.RGB
   B=VectorParameter() {parameterName=Cube Map Multiplier, defaultValue=<rgba 0.3751 0.489 0.5573 0>}.RGB
  ) {constB=1.5}
  n2 = ComponentMask(None=MaterialFunctionCall() {materialFunction=/Engine/Functions/GetPostProcessFunctions/GetAmbientCubemapTint}.AmbientCubemapTint) {b=true, g=true, r=true}
  Add(A=Desaturation(None=Multiply(A=@n1, B=@n2), Fraction=Constant() {r=0.2}))
  n3 = VectorParameter() {parameterName=waterColor, defaultValue=<rgba 0.0271 0.1967 0.4062 1>}
  n4 = CollectionParameter() {
   collection=&FortniteMaterialParameters
   parameterName=SunAndMoonModelDirectionalVector
  }
  n5 = Clamp(None=DotProduct(A=ComponentMask(None=@n4) {b=true, g=true, r=true}, B=ReflectionVectorWS()))
  n6 = ScalarParameter() {defaultValue=100, parameterName=Spec Highlight Power}
  n7 = TextureCoordinate()
  n8 = ScalarParameter() {defaultValue=2, parameterName=UVTIling}
  n9 = Multiply(
   A=MaterialFunctionCall!(ReflectionVector, ReflectionTexture) {materialFunction=/Engine/Functions/Engine_MaterialFunctions01/Reflections/ViewAlignedReflection}.UVs
   B=ScalarParameter() {defaultValue=1, parameterName=ViewDependantDistortion}
  )
  n10 = AppendVector(A=ScalarParameter() {parameterName=water_SpeedX}, B=ScalarParameter() {defaultValue=0.1, parameterName=water_SpeedY})
  n11 = TextureSample(UVs=StaticSwitchParameter(True=Multiply(A=@n7, B=@n8), False=Panner(Coordinate=@n9, Speed=@n10)) {parameterName=Water_UseUVs}) {samplerType=SAMPLERTYPE_LinearColor, texture=/Game/Characters/Player/Male/Medium/Heads/M_MED_AUX_Alien_Head_01/Textures/T_DistortionMask}
  n12 = TextureObject() {samplerType=SAMPLERTYPE_Masks, texture=/Game/Characters/Player/Male/Medium/Bodies/M_Med_Commando_AlienFishHead/Textures/T_SimpleBubble}
  n13 = ScalarParameter() {parameterName=Bubble Speed X}
  n14 = ScalarParameter() {defaultValue=0.1, parameterName=Bubble Speed Y_1}
  n15 = MaterialFunctionCall(Float3=MaterialFunctionCall!(
    TextureObject=@n12
    TextureSize=ScalarParameter() {defaultValue=10, parameterName=BubbleSize_A}
    WorldPosition
    Export Float 4
    In=AppendVector(A=@n13, B=@n14)
    World Space Normal
    ProjectionTransitionContrast
   ) {materialFunction=/Game/Athena/SafeZone/Preskinned_WP_Texture_ScalePan}.XYZ Texture) {materialFunction=&BreakOutFloat3Components}
  n16 = TextureObject() {samplerType=SAMPLERTYPE_Masks, texture=&T_SimpleBubble}
  n17 = ScalarParameter() {defaultValue=0.2, parameterName=Bubble Speed Y_2}
  n18 = MaterialFunctionCall(Float3=MaterialFunctionCall(TextureObject=@n16, TextureSize=ScalarParameter() {defaultValue=20, parameterName=BubbleSize_A_1}, In=AppendVector(A=@n13, B=@n17)) {materialFunction=&Preskinned_WP_Texture_ScalePan}.XYZ Texture) {materialFunction=&BreakOutFloat3Components}
  n19 = TextureObject() {samplerType=SAMPLERTYPE_Masks, texture=&T_SimpleBubble}
  n20 = MaterialFunctionCall(
   TextureObject=@n19
   TextureSize=ScalarParameter() {defaultValue=10, parameterName=BubbleSize_A_2}
   In=AppendVector(A=@n13, B=ScalarParameter() {defaultValue=0.4, parameterName=Bubble Speed Y_3})
  ) {materialFunction=&Preskinned_WP_Texture_ScalePan}
  n21 = Multiply(
   A=@n11.G
   B=Add(
    A=Add(A=@n15.R, B=@n18.G)
    B=MaterialFunctionCall(Float3=@n20.XYZ Texture) {materialFunction=&BreakOutFloat3Components}.B
   )
  )
  n22 = OneMinus(None=Clamp(None=Fresnel() {baseReflectFraction=0, exponent=1}))
  n23 = StaticSwitchParameter(True=Saturate(None=Multiply(A=@n21, B=@n22)), False=Constant()) {defaultValue=true, parameterName=BubbleVisibility}
  n24 = ScalarParameter() {defaultValue=0.4, parameterName="Bubble Highlight "}
  n25 = ScalarParameter() {defaultValue=0.1, parameterName=Bubble Highlight Thickness}
  n26 = Constant() {r=0.5}
  n27 = Multiply(
   A=VectorParameter() {parameterName=BubbleReflection, defaultValue=<rgba 5 5 5 1>}.RGB
   B=Multiply(A=@n1, B=SphereMask(A=@n23, B=@n24, Radius=@n25, Hardness=@n26))
  )
  n28 = Add(A=Multiply(A=Power(Base=@n5, Exp=@n6) {constExponent=20}, B=ScalarParameter() {defaultValue=8.5748, parameterName=Spec Intense}), B=@n27)
  n29 = QualitySwitch(Default=@n28, Low=Add(A=Constant(), B=@n27), High=@n28, Medium=@n28)
  n30 = ScalarParameter() {defaultValue=0.6, parameterName=Opac}
  n31 = ScalarParameter() {defaultValue=9.5587, parameterName=FadeDistance}
  n32 = LinearInterpolate(
   A=QualitySwitch(Default=DepthFade(Opacity=@n30, FadeDistance=@n31), Low=Constant())
   B=ScalarParameter() {defaultValue=0.5, parameterName=Bubble Opacity}
   Alpha=@n23
  )
  n33 = VectorParameter() {parameterName=Base Color, defaultValue=<rgba 0.2344 0.2414 0.2552 0>}
  n34 = ScalarParameter() {defaultValue=0.1, parameterName=Metallic}
  n35 = ScalarParameter() {defaultValue=1, parameterName=Specular}
  n36 = ScalarParameter() {defaultValue=0.1, parameterName=Roughness}
  n37 = QualitySwitch(Default=@n28, Low=Constant(), High=@n28, Medium=@n28)
  n38 = PreSkinnedPosition()
  n39 = MaterialFunctionCall(Float3=Divide(A=VertexInterpolator(VS=@n38).PS, B=ObjectBounds())) {materialFunction=&BreakOutFloat3Components}
  n40 = Multiply(A=AppendVector(A=@n39.G, B=@n39.G), B=ScalarParameter() {defaultValue=0.4, parameterName=WaterTiling})
  n41 = Normalize(VectorInput=Multiply(
    A=TextureSample(UVs=@n40) {samplerType=SAMPLERTYPE_Normal, texture=/Game/Global/Textures/Default/Noise/T_Noise_Normal_01}.RGB
    B=VectorParameter() {parameterName=Normal Intensity, defaultValue=<rgba 0.1 0.1 1 0>}.RGB
   ))
  n42 = LinearInterpolate(
   A=ScalarParameter() {parameterName=Fresnel Inner Transparency}
   B=ScalarParameter() {defaultValue=1, parameterName=Fresnel Outer Transparency}
   Alpha=Fresnel(ExponentIn=ScalarParameter() {defaultValue=4, parameterName=Fresnel Exponent}, Normal=Transform(None=@n41))
  )
  n43 = Constant3Vector() {constant=<rgba 0 0 1 0>}
  n44 = QualitySwitch(Default=@n41, Low=@n43, High=@n41, Medium=@n43)
  n45 = ScalarParameter() {defaultValue=1.25, parameterName=BowlThickness}
  n46 = VertexNormalWS()
  n47 = Clamp(None=Multiply(A=Fresnel(ExponentIn=@n45, Normal=@n46) {baseReflectFraction=0}, B=ScalarParameter() {defaultValue=1.5, parameterName=Boost}))
  n48 = ScalarParameter() {defaultValue=0.7, parameterName=TransitionPower}
  n49 = BreakMaterialAttributes!(Attr=BlendMaterialAttributes(
    A=MakeMaterialAttributes!(
     BaseColor=@n3.RGB
     Metallic
     Specular
     Roughness
     Anisotropy
     EmissiveColor=@n29
     Opacity=@n32
     OpacityMask
     Normal
     Tangent
     WorldPositionOffset
     SubsurfaceColor
     ClearCoat
     ClearCoatRoughness
     AmbientOcclusion
     Refraction
     CustomizedUVs_0
     CustomizedUVs_1
     CustomizedUVs_2
     CustomizedUVs_3
     CustomizedUVs_4
     CustomizedUVs_5
     CustomizedUVs_6
     CustomizedUVs_7
     PixelDepthOffset
     ShadingModel
     Displacement
    )
    B=MakeMaterialAttributes(BaseColor=@n33.RGB, Metallic=@n34, Specular=@n35, Roughness=@n36, EmissiveColor=@n37, Opacity=@n42, OpacityMask=@n42, Normal=@n44)
    Alpha=Power(Base=@n47, Exp=@n48) {constExponent=14}
   ) {vertexAttributeBlendType=UseB})
  n50 = Clamp(None=LinearInterpolate(
    A=ScalarParameter() {defaultValue=1, parameterName=Refraction_2, sliderMax=2}
    B=ScalarParameter() {defaultValue=0.8, parameterName=Refraction_Bubble, sliderMax=2}
    Alpha=@n23
   ))
  n51 = Constant()
  n52 = SetMaterialAttributes!(
   MaterialAttributes
   Base Color=@n49.BaseColor
   Emissive Color=@n49.BaseColor
   Opacity=Multiply(
    A=MaterialFunctionCall!(Animation Gradient, Rez In Percentage) {materialFunction=/Game/Packages/Fortress_SharedMaterials/Base_Material_Functions/CharacterResInEffect}.Modulative Opacity
    B=@n49.Opacity
   )
   Normal=@n49.BaseColor
   World Position Offset=@n49.BaseColor
   Ambient Occlusion=@n49.BaseColor
   Refraction=QualitySwitch(Default=@n50, Low=@n51, High=@n50, Medium=@n51)
  ) {attributeSetTypes [
    #1
    #2
    #3
    #4
    #5
    #6
    #7
   ]}
  MaterialFunctionCall(In=@n52) {materialFunction=&MF_TechArtOverrides}
 ]
}
